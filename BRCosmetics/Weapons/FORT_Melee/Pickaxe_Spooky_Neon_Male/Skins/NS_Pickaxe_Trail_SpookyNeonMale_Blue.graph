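# Niagara system for the blue swing trail on the Spooky Neon Male pickaxe skin.
# One anim-trail emitter spawns short-lived ribbon particles using bone or socket indices, with a skeletal mesh exposed as a user parameter.
# Two ribbon renderers draw it, with a separate color and width set for the secondary ribbon.
# Curves shape the trail update and alpha fade, and time-of-day scalars dim opacity, most strongly at night and evening.
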NiagaraSystem /BRCosmetics/Weapons/FORT_Melee/Pickaxe_Spooky_Neon_Male/Skins/NS_Pickaxe_Trail_SpookyNeonMale_Blue
dependencies {usedDataInterfaces [&NiagaraDataInterfaceCurve], usedDynamicInputs [], usedModules [&SystemState]}
emitters {FN_AnimTrail_V4 {
  emitterUpdateScript [
   &EmitterState()
   /Game/Effects/Niagara/Modules/AnimTrail/FN_SpawnAnimTrail/FN_SpawnAnimTrail(Number of Anim Trails=1, SpawnRate=40)
  ]
  particleSpawnScript [
   /Game/Effects/Niagara/Modules/AnimTrail/FN_InitializeAnimTrail/FN_InitializeAnimTrailBoneIndecies(First Bone or Socket Index=1)
   &FN_InitializeAnimTrail(Color=<rgba 0 0.1492 0.5162 1>, Lifetime=0.175) {module=FN_InitializeAnimTrail001}
   &AddVelocity(Velocity=<0 2000 0>) {enabled=false}
  ]
  particleUpdateScript [
   /SetVariables(
    Particles.Ribbon2_Color:LinearColor=&MultiplyLinearColor(
     Linear Color=<rgba 0.4139 0.4963 0.6483 0.25>
     Scale Factor=&MakeVector4(
      X=1
      Y=1
      Z=1
      W=&FNTimeOfDayScalar(Night Scale=0.3, Morning Scale=0.5, Day Scale=0.5, Evening Scale=0.3)
     )
    )
    Particles.Ribbon2_Width=10
   ) {isSetParameters=true}
   &SolveForcesAndVelocity(Clamp Velocity=false, Limit Acceleration=false)
   /Game/Effects/Niagara/Modules/AnimTrail/FN_UpdateAnimTrail/FN_UpdateAnimTrail(New float:NiagaraFloat=&FloatFromCurve(FloatCurve=DataInterface() {propertyValues {Curve {keys [
         {arriveTangent=9.0725, interpMode=RCIM_Cubic, leaveTangent=9.0725, tangentMode=RCTM_User, time=0, value=0}
         {arriveTangent=-2.5249, interpMode=RCIM_Cubic, leaveTangent=-2.5249, tangentMode=RCTM_User, time=0.1408, value=0.9847}
         {arriveTangent=-0.0405, interpMode=RCIM_Cubic, leaveTangent=-0.0405, tangentMode=RCTM_User, time=1, value=0}
        ]}}}))
   &ScaleColor(
    Scale Alpha=&FloatFromCurve(
     FloatCurve=DataInterface() {propertyValues {Curve {keys [
         {interpMode=RCIM_Cubic, tangentMode=RCTM_User, time=0.25, value=0}
         {interpMode=RCIM_Cubic, tangentMode=RCTM_User, time=0.6, value=1}
         {tangentMode=RCTM_User, time=0.75, value=0}
        ]}}}
     Scale Curve=&FNTimeOfDayScalar(Night Scale=0.7, Morning Scale=1, Evening Scale=0.7)
    )
    ScaleA=true
    ScaleRGB=true
   )
  ]
  rendererClasses [&NiagaraRibbonRendererProperties]
  renderers [
   {rendererClass=&NiagaraRibbonRendererProperties}
   {rendererClass=&NiagaraRibbonRendererProperties, rendererIndex=1}
  ]
 }}
summary {
 systemName=NS_Pickaxe_Trail_SpookyNeonMale_Blue
 emitters [{emitterName=FN_AnimTrail_V4, rendererClasses [&NiagaraRibbonRendererProperties]}]
 userVariables [
  User.RibbonIDOffset {type=NiagaraInt32}
  User.SkeletalMesh {
   type=NiagaraDataInterfaceSkeletalMesh
   defaultValue {
    struct=NiagaraExt_VariableValue_DataInterface
    value {dataInterface=&self:NiagaraDataInterfaceSkeletalMesh_0}
   }
  }
 ]
}
